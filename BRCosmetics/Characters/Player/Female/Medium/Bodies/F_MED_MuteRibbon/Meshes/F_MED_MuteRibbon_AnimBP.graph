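# Post-copy animation Blueprint for the F_MED_MuteRibbon female body, adding rig deformation and front-end attachment offsets.
# It copies the pose from the attached parent mesh, then runs it through the body's DeformRig Control Rig.
# Additive bone-space translations shift the backpack and cape attach bones, weighted by a shared alpha.
# When in the front end, each update sets that alpha from the front-end anim instance's Frontend_Idle_Bypass curve.
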AnimBlueprint /BRCosmetics/Characters/Player/Female/Medium/Bodies/F_MED_MuteRibbon/Meshes/F_MED_MuteRibbon_AnimBP
target_skeleton=&/F_MED_MuteRibbon_Skeleton
graphs [
 AnimGraph {nodes [
   n1 = Control Rig(Source=Copy Pose From Mesh() {bUseAttachedParent=true}) {
    alphaBoolBlend {blendOption=Linear}
    controlRigAssetReference {blueprintRigClass=&/F_MED_MuteRibbon_DeformRig.F_MED_MuteRibbon_DeformRig_C}
   }
   n2 = Get Custom FE Idle is Playing() -> (Custom FE Idle is Playing:double)
   n3 = "Transform (Modify) Bone - Bone: attach_backpack"(ComponentPose=Local To Component(LocalPose=@n1), Alpha=@n2, Translation="0,-5,0", Rotation="0,0,0", Scale="1,1,1") {
    translationMode=BMM_Additive
    translationSpace=BCS_BoneSpace
    alphaBoolBlend {blendOption=Linear}
    boneToModify {boneName=attach_backpack}
    translation {y=-5}
   }
   n4 = "Transform (Modify) Bone - Bone: attach_cape"(ComponentPose=@n3, Alpha=@n2, Translation="0,-5,0", Rotation="0,0,0", Scale="1,1,1") {
    translationMode=BMM_Additive
    translationSpace=BCS_BoneSpace
    alphaBoolBlend {blendOption=Linear}
    boneToModify {boneName=attach_cape}
    translation {y=-5}
   }
   Output Pose(Result=Component To Local(ComponentPose=@n4)) {layerGroup=None}
  ]}
 EventGraph {nodes [
   Event BlueprintUpdateAnimation()
   > Branch(Condition=Get bIsInFrontEnd())
   > Set Custom FE Idle is Playing(Custom FE Idle is Playing:double=GetCurveValue(self=GetOwnerFrontEndAnimInstance(), CurveName=Frontend_Idle_Bypass)) -> (Output_Get:double)
  ]}
]
variables [Custom FE Idle is Playing {type=number}]
