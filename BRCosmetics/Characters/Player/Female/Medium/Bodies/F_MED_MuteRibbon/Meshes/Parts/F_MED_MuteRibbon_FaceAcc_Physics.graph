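# Physics asset for the FaceAcc part of the female medium MuteRibbon player cosmetic, setting up hair physics for bangs, spikes and ponytail.
# Capsule bodies form a three-link bang chain, a six-link ponytail and two single spikes, all rooted on the head, which links through neck_02 to spine_05.
# Hair joints have zero twist limits and narrow swing, except the ponytail's much wider second swing axis; head and neck joints rotate freely.
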PhysicsAsset /BRCosmetics/Characters/Player/Female/Medium/Bodies/F_MED_MuteRibbon/Meshes/Parts/F_MED_MuteRibbon_FaceAcc_Physics
bodies {
 dyn_bang_01 [{shapeType=Capsule, length=2.75, radius=4.25, center=<2.228 0 1.4373>, rotation=<rot -85 180 -180>}]
 dyn_bang_02 [{shapeType=Capsule, length=2.5, radius=3.75, center=<3.038 0 0.7411>, rotation=<rot -79.9999 0 0>}]
 dyn_bang_03 [{shapeType=Capsule, length=2.5, radius=1.75, center=<1.6413 0 -0.1436>, rotation=<rot -85 179.9999 -180>}]
 dyn_hairSpike_bot [{shapeType=Capsule, length=2.75, radius=2.25, center=<1.4692 0 0>, rotation=<rot -90 166.7373 -166.7374>}]
 dyn_hairSpike_top [{shapeType=Capsule, length=1, radius=2.25, center=<1.4519 0 0>, rotation=<rot -90 0.0001 -0.0003>}]
 dyn_ponytail_01 [{shapeType=Capsule, length=3.75, radius=4.5, center=<3.9408 0 2.1686>, rotation=<rot -80.0001 0 0>}]
 dyn_ponytail_02 [{shapeType=Capsule, length=1.5, radius=6.75, center=<4.2039 0 0.3678>, rotation=<rot -85.0002 0 0>}]
 dyn_ponytail_03 [{shapeType=Capsule, length=2.25, radius=8.25, center=<4.4453 0 -1.0012>, rotation=<rot -80.0003 0 0>}]
 dyn_ponytail_04 [{shapeType=Capsule, length=3.5, radius=7.75, center=<4.2039 0 0.3678>, rotation=<rot -85.0002 0 0>}]
 dyn_ponytail_05 [{shapeType=Capsule, length=6.5, radius=7, center=<4.2039 0 -0.8053>, rotation=<rot -80 -180 180>}]
 dyn_ponytail_06 [{shapeType=Capsule, length=7.5, radius=6.25, center=<3.1284 0 0.2737>, rotation=<rot -85.0002 0 0>}]
 head [{shapeType=Capsule, length=4.5, radius=9.75, center=<9.2652 3.186 0>, rotation=<rot 90 -179.9983 -179.9979>}]
 neck_02 [{shapeType=Capsule, length=1, radius=1}]
 spine_05 [{shapeType=Capsule, length=17, radius=11.505, shapeName=spine_05_capsule, center=<0 -1.2786 0>, rotation=<rot 0 -90 0>}]
}
constraints [
 {bone1Name=dyn_bang_02, bone2Name=dyn_bang_01, swing1LimitDegrees=10, swing2LimitDegrees=15, twistLimitDegrees=0}
 {bone1Name=dyn_bang_03, bone2Name=dyn_bang_02, swing1LimitDegrees=10, swing2LimitDegrees=15, twistLimitDegrees=0}
 {bone1Name=dyn_ponytail_06, bone2Name=dyn_ponytail_05, swing1LimitDegrees=10, swing2LimitDegrees=60, twistLimitDegrees=0}
 {bone1Name=dyn_ponytail_02, bone2Name=dyn_ponytail_01, swing1LimitDegrees=0, swing2LimitDegrees=50, twistLimitDegrees=0}
 {bone1Name=dyn_ponytail_03, bone2Name=dyn_ponytail_02, swing1LimitDegrees=5, swing2LimitDegrees=70, twistLimitDegrees=0}
 {bone1Name=dyn_ponytail_04, bone2Name=dyn_ponytail_03, swing1LimitDegrees=10, swing2LimitDegrees=70, twistLimitDegrees=0}
 {bone1Name=dyn_ponytail_05, bone2Name=dyn_ponytail_04, swing1LimitDegrees=10, swing2LimitDegrees=70, twistLimitDegrees=0}
 {bone1Name=dyn_ponytail_01, bone2Name=head, swing1LimitDegrees=0, swing2LimitDegrees=40, twistLimitDegrees=0}
 {bone1Name=dyn_hairSpike_top, bone2Name=head, swing1LimitDegrees=10, swing2LimitDegrees=10, twistLimitDegrees=0}
 {bone1Name=dyn_hairSpike_bot, bone2Name=head, swing1LimitDegrees=10, swing2LimitDegrees=10, twistLimitDegrees=0}
 {bone1Name=dyn_bang_01, bone2Name=head, swing1LimitDegrees=10, swing2LimitDegrees=15, twistLimitDegrees=0}
 {
  bone1Name=head
  bone2Name=neck_02
  swing1LimitDegrees=45
  swing1Motion=Free
  swing2LimitDegrees=45
  swing2Motion=Free
  twistLimitDegrees=45
  twistMotion=Free
 }
 {
  bone1Name=neck_02
  bone2Name=spine_05
  swing1LimitDegrees=45
  swing1Motion=Free
  swing2LimitDegrees=45
  swing2Motion=Free
  twistLimitDegrees=45
  twistMotion=Free
 }
]
