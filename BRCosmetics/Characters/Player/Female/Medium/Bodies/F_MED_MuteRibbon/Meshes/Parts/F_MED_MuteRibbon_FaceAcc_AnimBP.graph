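# Anim Blueprint for the MuteRibbon female outfit's face accessory, posing and simulating its dyn_ponytail bones.
# The output path copies the parent mesh pose, runs RigidBody in neck_02 space, then offsets and scales the bones, pinning the last two to spine_05.
# A second transform chain and a neck-driven layered blend are disconnected.
# The event graph computes gravity, backbling, sprint and neck-bend values, none visibly wired into the output pose.
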AnimBlueprint /BRCosmetics/Characters/Player/Female/Medium/Bodies/F_MED_MuteRibbon/Meshes/Parts/F_MED_MuteRibbon_FaceAcc_AnimBP
target_skeleton=&/F_MED_MuteRibbon_FaceAcc_Skeleton
graphs [
 AnimGraph {nodes [
   n1 = "Constraint - dyn_ponytail_06 (T)"(Alpha=1, ConstraintWeights_0=1) {
    alphaBoolBlend {blendOption=Linear}
    boneToModify {boneName=dyn_ponytail_06}
    constraintSetup [{transformType=Translation, targetBone {boneName=spine_05}}]
    constraintWeights [1]
   }
   "Constraint - dyn_ponytail_05 (T)"(ComponentPose=@n1, Alpha=1, ConstraintWeights_0=1) {
    alphaBoolBlend {blendOption=Linear}
    boneToModify {boneName=dyn_ponytail_05}
    constraintSetup [{transformType=Translation, targetBone {boneName=spine_05}}]
    constraintWeights [1]
   }
   Layered blend per bone!(
    BasePose
    BlendPoses_0
    BlendPoses_1:Pose Link Structure
    BlendWeights_0=Get neckRotateBack() -> (neckRotateBack:double)
    BlendWeights_1:float=1
   ) -> (Pose) {
    blendPoses [{linkId=-1, sourceLinkId=-1}, {linkId=-1, sourceLinkId=-1}]
    blendWeights [1,1]
    layerSetup [{branchFilters [{blendDepth=1, boneName=neck_01}]}, {branchFilters []}]
   }
   n2 = "Transform (Modify) Bone - Bone: dyn_ponytail_01"(Alpha=1, Translation="0,0,0", Rotation="0.000,0.000,-25.000", Scale="1,1,1") {
    rotationMode=BMM_Additive
    scaleMode=BMM_Replace
    translationMode=BMM_Additive
    alphaBoolBlend {blendOption=Linear}
    boneToModify {boneName=dyn_ponytail_01}
    rotation {roll=-25}
   }
   n3 = "Transform (Modify) Bone - Bone: dyn_ponytail_02"(ComponentPose=@n2, Alpha=1, Translation="0,0,0", Rotation="0.000,0.000,-10.000", Scale="1,0.8,1") {
    rotationMode=BMM_Additive
    scaleMode=BMM_Replace
    translationMode=BMM_Additive
    alphaBoolBlend {blendOption=Linear}
    boneToModify {boneName=dyn_ponytail_02}
    rotation {roll=-10}
    scale {y=0.8}
   }
   n4 = "Transform (Modify) Bone - Bone: dyn_ponytail_03"(ComponentPose=@n3, Alpha=1, Translation="-0.000,-0.000,-1.958", Rotation="0.000,0.000,-10.000", Scale="1,1,1") {
    rotationMode=BMM_Additive
    scaleMode=BMM_Replace
    translationMode=BMM_Additive
    alphaBoolBlend {blendOption=Linear}
    boneToModify {boneName=dyn_ponytail_03}
    rotation {roll=-10}
    translation {z=-1.958}
   }
   n5 = "Transform (Modify) Bone - Bone: dyn_ponytail_04"(ComponentPose=@n4, Alpha=1, Translation="-0.000,7.038,4.883", Rotation="0,0,0", Scale="1,1,0.8") {
    rotationMode=BMM_Additive
    scaleMode=BMM_Replace
    translationMode=BMM_Additive
    alphaBoolBlend {blendOption=Linear}
    boneToModify {boneName=dyn_ponytail_04}
    scale {z=0.8}
    translation {y=7.038, z=4.883}
   }
   n6 = "Constraint - dyn_ponytail_05 (T)"(ComponentPose=@n5, Alpha=1, ConstraintWeights_0=1) {
    alphaBoolBlend {blendOption=Linear}
    boneToModify {boneName=dyn_ponytail_05}
    constraintSetup [{transformType=Translation, targetBone {boneName=spine_05}}]
    constraintWeights [1]
   }
   n7 = "Constraint - dyn_ponytail_05 (R)"(ComponentPose=@n6, Alpha=1, ConstraintWeights_0=1) {
    alphaBoolBlend {blendOption=Linear}
    boneToModify {boneName=dyn_ponytail_05}
    constraintSetup [{transformType=Rotation, targetBone {boneName=spine_05}}]
    constraintWeights [1]
   }
   n8 = "Constraint - dyn_ponytail_06 (T)"(ComponentPose=@n7, Alpha=1, ConstraintWeights_0=1) {
    alphaBoolBlend {blendOption=Linear}
    boneToModify {boneName=dyn_ponytail_06}
    constraintSetup [{transformType=Translation, targetBone {boneName=spine_05}}]
    constraintWeights [1]
   }
   n9 = "Constraint - dyn_ponytail_06 (R)"(ComponentPose=@n8, Alpha=1, ConstraintWeights_0=1) {
    alphaBoolBlend {blendOption=Linear}
    boneToModify {boneName=dyn_ponytail_06}
    constraintSetup [{transformType=Rotation, targetBone {boneName=spine_05}}]
    constraintWeights [1]
   }
   n10 = "Transform (Modify) Bone - Bone: dyn_ponytail_05"(ComponentPose=@n9, Alpha=1, Translation="0,20,-2.988", Rotation="0.000,0.000,-50.000", Scale="0.75,1,0.6") {
    rotationMode=BMM_Additive
    scaleMode=BMM_Replace
    translationMode=BMM_Additive
    alphaBoolBlend {blendOption=Linear}
    boneToModify {boneName=dyn_ponytail_05}
    rotation {roll=-50}
    scale {x=0.75, z=0.6}
    translation {y=20, z=-2.988}
   }
   "Transform (Modify) Bone - Bone: dyn_ponytail_06"(ComponentPose=@n10, Alpha=1, Translation="0.000,-6.825,6.608", Rotation="0.000,0.000,-25.000", Scale="1.000,1.000,0.500") {
    rotationMode=BMM_Additive
    scaleMode=BMM_Replace
    translationMode=BMM_Additive
    alphaBoolBlend {blendOption=Linear}
    boneToModify {boneName=dyn_ponytail_06}
    rotation {roll=-25}
    scale {z=0.5}
    translation {y=-6.825, z=6.608}
   }
   n11 = RigidBody!(
    ComponentPose=Local To Component(LocalPose=Copy Pose From Mesh() {bUseAttachedParent=true})
    Alpha
    OverrideWorldGravity:Vector="0,0,-980"
    ExternalForce
   ) -> (Pose) {
    bClampLinearTranslationLimitToRefPose=true
    bOverrideWorldGravity=true
    overridePhysicsAsset=&/F_MED_MuteRibbon_FaceAcc_Physics
    simulationSpace=BaseBoneSpace
    alphaBoolBlend {blendOption=Linear}
    alphaScaleBiasClamp {bInterpResult=true, interpSpeedDecreasing=1, interpSpeedIncreasing=1}
    baseBoneRef {boneName=neck_02}
    componentAppliedLinearAccClamp=<3000 3000 3000>
    componentLinearAccScale=<0.15 0.5 0.5>
    componentLinearVelScale=<0.4 0.5 0.5>
    overrideWorldGravity {z=-980}
    simSpaceSettings {maxAngularAcceleration=3000, maxAngularVelocity=1000, maxLinearAcceleration=3000, maxLinearVelocity=1000}
   }
   n12 = "Transform (Modify) Bone - Bone: dyn_ponytail_01"(ComponentPose=@n11, Alpha=1, Translation="0,0,0", Rotation="-15.000,0.001,0.012", Scale="1,1,1") {
    rotationMode=BMM_Additive
    rotationSpace=BCS_BoneSpace
    scaleMode=BMM_Replace
    scaleSpace=BCS_BoneSpace
    translationMode=BMM_Additive
    translationSpace=BCS_BoneSpace
    alphaBoolBlend {blendOption=Linear}
    boneToModify {boneName=dyn_ponytail_01}
    rotation=<rot -15.0001 0.001 0.012>
   }
   n13 = "Transform (Modify) Bone - Bone: dyn_ponytail_02"(ComponentPose=@n12, Alpha=1, Translation="-3.523,0,-1.913", Rotation="-0.000,-0.055,0.032", Scale="1,1,1") {
    rotationMode=BMM_Additive
    rotationSpace=BCS_BoneSpace
    scaleMode=BMM_Replace
    scaleSpace=BCS_BoneSpace
    translationMode=BMM_Additive
    translationSpace=BCS_BoneSpace
    alphaBoolBlend {blendOption=Linear}
    boneToModify {boneName=dyn_ponytail_02}
    rotation=<rot -0.0001 -0.055 0.032>
    translation {x=-3.523, z=-1.913}
   }
   n14 = "Transform (Modify) Bone - Bone: dyn_ponytail_03"(ComponentPose=@n13, Alpha=1, Translation="7.372,3.651,-5.322", Rotation="-40.000,-0.011,0.032", Scale="1,1,0.8") {
    rotationMode=BMM_Additive
    rotationSpace=BCS_BoneSpace
    scaleMode=BMM_Replace
    scaleSpace=BCS_BoneSpace
    translationMode=BMM_Additive
    translationSpace=BCS_BoneSpace
    alphaBoolBlend {blendOption=Linear}
    boneToModify {boneName=dyn_ponytail_03}
    rotation=<rot -40 -0.011 0.032>
    scale {z=0.8}
    translation=<7.372 3.651 -5.322>
   }
   n15 = "Transform (Modify) Bone - Bone: dyn_ponytail_04"(ComponentPose=@n14, Alpha=1, Translation="2.618,0.317,-0.146", Rotation="10.000,0.004,-0.041", Scale="1,1,0.8") {
    rotationMode=BMM_Additive
    rotationSpace=BCS_BoneSpace
    scaleMode=BMM_Replace
    scaleSpace=BCS_BoneSpace
    translationMode=BMM_Additive
    translationSpace=BCS_BoneSpace
    alphaBoolBlend {blendOption=Linear}
    boneToModify {boneName=dyn_ponytail_04}
    rotation=<rot 10 0.004 -0.041>
    scale {z=0.8}
    translation=<2.618 0.317 -0.146>
   }
   n16 = "Constraint - dyn_ponytail_05 (P)"(ComponentPose=@n15, Alpha=1, ConstraintWeights_0=1) {
    alphaBoolBlend {blendOption=Linear}
    boneToModify {boneName=dyn_ponytail_05}
    constraintSetup [{targetBone {boneName=spine_05}}]
    constraintWeights [1]
   }
   n17 = "Transform (Modify) Bone - Bone: dyn_ponytail_05"(ComponentPose=@n16, Alpha=1, Translation="6.939,0,-20.189", Rotation="-40.000,-0.128,0.057", Scale="0.500,1.000,0.500") {
    rotationMode=BMM_Additive
    rotationSpace=BCS_BoneSpace
    scaleMode=BMM_Replace
    scaleSpace=BCS_BoneSpace
    translationMode=BMM_Additive
    translationSpace=BCS_BoneSpace
    alphaBoolBlend {blendOption=Linear}
    boneToModify {boneName=dyn_ponytail_05}
    rotation=<rot -40 -0.128 0.057>
    scale {x=0.5, z=0.5}
    translation {x=6.939, z=-20.189}
   }
   n18 = "Constraint - dyn_ponytail_06 (P)"(ComponentPose=@n17, Alpha=1, ConstraintWeights_0=1) {
    alphaBoolBlend {blendOption=Linear}
    boneToModify {boneName=dyn_ponytail_06}
    constraintSetup [{targetBone {boneName=spine_05}}]
    constraintWeights [1]
   }
   n19 = "Transform (Modify) Bone - Bone: dyn_ponytail_06"(ComponentPose=@n18, Alpha=1, Translation="-20.652,0,-16.414", Rotation="-75.001,-0.752,0.715", Scale="0.3,1,0.3") {
    rotationMode=BMM_Additive
    rotationSpace=BCS_BoneSpace
    scaleMode=BMM_Replace
    scaleSpace=BCS_BoneSpace
    translationMode=BMM_Additive
    translationSpace=BCS_BoneSpace
    alphaBoolBlend {blendOption=Linear}
    boneToModify {boneName=dyn_ponytail_06}
    rotation=<rot -75.001 -0.7519 0.7149>
    scale {x=0.3, z=0.3}
    translation {x=-20.652, z=-16.414}
   }
   Output Pose(Result=Component To Local(ComponentPose=@n19)) {layerGroup=None}
  ]}
 EventGraph {nodes [
   Event BlueprintUpdateAnimation()
   > n1 = Gravity Override State Function DEPRECATED(CharacterPartAnimInstance=Self-Reference(), GameTimeInSeconds=GetGameTimeInSeconds(), GravityOverrideParams=Get Gravity Override Params())
   > Set Gravity Override(Gravity Override=@n1)
   n2 = CheckForBackbling()
   > n3 = Branch(Condition=@n2)
   > Set backbling alpha(backbling alpha:double=1) -> (Output_Get:double)
   > n4 = Set rban alpha(rban alpha:double=0) -> (Output_Get:double)
   n5 = Branch(execute=[@n10.then, @n9.then, @n4.then], Condition=@n2)
   n6 = GetOwnerSkeletalMeshComponent()
   n7 = BreakRotator(InRot=BreakTransform(InTransform=GetDeltaTransformFromRefPose(self=@n6, BoneName=neck_01, BaseName=spine_05)).Rotation)
   > Set neckRotateBack(neckRotateBack:double=float * float(A=MapRangeClamped(Value=@n7.Yaw, InRangeA=0, InRangeB=-20), B=0.05) {is=title K2Node_PromotableOperator}) -> (Output_Get:double)
   Set neckRotateBack(execute=@n5.else, neckRotateBack=0)
   Set backbling alpha(execute=@n3.else, backbling alpha=0)
   > Set rban alpha(rban alpha=0.9)
   > n8 = Branch(Condition=Get bIsTacticalSprinting())
   > n9 = Set rban alpha(rban alpha=0.7)
   n10 = Set rban alpha(execute=@n8.else, rban alpha=0.9)
   AND Boolean(A=false, B=OR Boolean(A=Get bIsJumping(), B=Get bIsFalling()))
   Get PawnVelocityZ()
  ]}
]
variables [
 backbling alpha {type=number}
 Gravity Override {type=Vector}
 Gravity Override Params {type=GravityOverrideParamsStruct}
 neckRotateBack {type=number}
 rban alpha {type=number}
]
